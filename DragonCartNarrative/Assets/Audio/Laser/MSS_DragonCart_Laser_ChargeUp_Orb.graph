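# Mono one-shot MetaSound for the DragonCart orb laser's charge-up, using the orb laser's third-person fire attenuation.
# On play, it picks a wave from the Sounds array and applies random pitch and volume offsets.
# A trigger delayed by ChargeDuration drops an interpolated gain target to zero, fading the sound out.
# The AudioWorldization enclosure modulation bus then drives a low-pass cutoff and final gain, muffling and quieting the sound as its value rises.
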MetaSoundSource /DragonCartNarrative/Assets/Audio/Laser/MSS_DragonCart_Laser_ChargeUp_Orb
inputs [
 ChargeDuration {type=Float}
 Random.Pitch.Base {type=Float}
 Random.Pitch.Spread {type=Float}
 Random.Volume.Base {type=Float}
 Random.Volume.Spread {type=Float}
 Sounds {type=WaveAsset:Array}
 UE.Source.OnPlay {type=Trigger}
]
interfaces [UE.OutputFormat.Mono:1.0, UE.Source.OneShot:1.0, UE.Source:1.1]
outputs [
 "UE.OutputFormat.Mono.Audio:0" {type=Audio}
 UE.Source.OneShot.OnFinished {type=Trigger}
]
pages [Default {
  input_defaults {
   ChargeDuration=2.132
   Random.Pitch.Base=0
   Random.Pitch.Spread=1
   Random.Volume.Base=1
   Random.Volume.Spread=0
   UE.Source.OnPlay=null
   Sounds [/DragonCartNarrative/Assets/Audio/Laser/WAV/DragonCart_Orb_Laser_ChargeUp_Orb_01]
  }
  nodes [
   n1 = Modulation.GetModulatorValue() {also {defaults {Modulator=&AudioWorldization_Bus_Enclosure}}}
   ArrayRandomGet_0 = Array.Random Get:WaveAsset:Array(In Array=^Sounds, Next=^UE.Source.OnPlay)
   TriggerDelay_0 = UE.Trigger Delay(In=@ArrayRandomGet_0.On Next, Delay Time=UE.ConversionFloatToTime(In=^ChargeDuration).Out)
   n2 = Value.Value:Float(Set=@TriggerDelay_0.Out) {also {defaults {Init Value=1, Target Value=0}}}
   n3 = AbsoluteValue.Abs:Float(Input=^Random.Volume.Spread)
   DecibelstoLinearGain_0 = UE.Decibels to Linear Gain:Float(Decibels=UE.RandomFloat(
     Next=^UE.Source.OnPlay
     Max=UE.Add:Float(PrimaryOperand=^Random.Volume.Base, AdditionalOperands=@n3.Value).Out
     Min=UE.Subtract:Float(PrimaryOperand=^Random.Volume.Base, AdditionalOperands=@n3.Value).Out
    ).Value)
   n4 = AbsoluteValue.Abs:Float(Input=^Random.Pitch.Spread)
   WavePlayer_0 = UE.Wave Player:Mono(
    Play=@ArrayRandomGet_0.On Next
    Wave Asset=@ArrayRandomGet_0.Value
    Pitch Shift=UE.RandomFloat(
     Next=^UE.Source.OnPlay
     Max=UE.Add:Float(PrimaryOperand=^Random.Pitch.Base, AdditionalOperands=@n4.Value).Out
     Min=UE.Subtract:Float(PrimaryOperand=^Random.Pitch.Base, AdditionalOperands=@n4.Value).Out
    ).Value
   ) {also {version [1,1]}}
   OnePoleLowPassFilter_0 = UE.One-Pole Low Pass Filter:Audio(
    Cutoff Frequency=MapRange.MapRange:Float(In=@n1.Out) {also {defaults {In Range A=0.75, In Range B=0.9, Out Range A=20000, Out Range B=150}}}.Out Value
    In=UE.Multiply:Audio by Float(
     AdditionalOperands=UE.InterpTo:Audio(Target=@n2.Output Value) {also {defaults {Interp Time=0.5}}}.Value
     PrimaryOperand=UE.Multiply:Audio by Float(AdditionalOperands=@DecibelstoLinearGain_0.Linear Gain, PrimaryOperand=@WavePlayer_0.Out Mono).Out
    ).Out
   )
   Multiply_0 = UE.Multiply:Audio by Float(
    AdditionalOperands=MapRange.MapRange:Float(In=@n1.Out) {also {defaults {In Range A=0.75, In Range B=0.9, Out Range A=1, Out Range B=0.2}}}.Out Value
    PrimaryOperand=@OnePoleLowPassFilter_0.Out
   )
  ]
  links [{dst="@UE.OutputFormat.Mono.Audio:0", src=Multiply_0.Out}, {dst="@UE.Source.OneShot.OnFinished", src=WavePlayer_0.On Finished}]
 }]
properties {
 assetImportData=&self:AssetImportData
 attenuationSettings=&/ATT_DragonCart_Orb_Laser_Fire_3P
 maxDistance=84000
 sampleRate=48000
 soundClassObject=&Fort_Events_Gameplay
}
