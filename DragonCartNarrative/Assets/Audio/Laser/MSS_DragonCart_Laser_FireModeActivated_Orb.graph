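# One-shot mono MetaSound source that plays when the DragonCart orb laser's fire mode activates, spatialized with the third-person laser fire attenuation.
# On play it picks a random shooting-mode-start wave and applies base ± spread pitch and decibel volume, with both spreads zero by default.
# The AudioWorldization enclosure modulator then drives a one-pole low-pass and gain drop, muffling the sound as enclosure rises.
# The source also sends to the OrbMarkingActivate audio bus.
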MetaSoundSource /DragonCartNarrative/Assets/Audio/Laser/MSS_DragonCart_Laser_FireModeActivated_Orb
inputs [
 Random.Pitch.Base {type=Float}
 Random.Pitch.Spread {type=Float}
 Random.Volume.Base {type=Float}
 Random.Volume.Spread {type=Float}
 Sounds {type=WaveAsset:Array}
 UE.Source.OnPlay {type=Trigger}
]
interfaces [UE.OutputFormat.Mono:1.0, UE.Source.OneShot:1.0, UE.Source:1.1]
outputs [
 "UE.OutputFormat.Mono.Audio:0" {type=Audio}
 UE.Source.OneShot.OnFinished {type=Trigger}
]
pages [Default {
  input_defaults {
   Random.Pitch.Base=0
   Random.Pitch.Spread=0
   Random.Volume.Base=6
   Random.Volume.Spread=0
   UE.Source.OnPlay=null
   Sounds [
    /DragonCartNarrative/Assets/Audio/Laser/WAV/DragonCart_Orb_Laser_ShootingModeStart_01
    /DragonCartNarrative/Assets/Audio/Laser/WAV/DragonCart_Orb_Laser_ShootingModeStart_02
    /DragonCartNarrative/Assets/Audio/Laser/WAV/DragonCart_Orb_Laser_ShootingModeStart_03
    /DragonCartNarrative/Assets/Audio/Laser/WAV/DragonCart_Orb_Laser_ShootingModeStart_04
   ]
  }
  nodes [
   n1 = Modulation.GetModulatorValue() {also {defaults {Modulator=&AudioWorldization_Bus_Enclosure}}}
   n2 = AbsoluteValue.Abs:Float(Input=^Random.Volume.Spread)
   DecibelstoLinearGain_0 = UE.Decibels to Linear Gain:Float(Decibels=UE.RandomFloat(
     Next=^UE.Source.OnPlay
     Max=UE.Add:Float(PrimaryOperand=^Random.Volume.Base, AdditionalOperands=@n2.Value).Out
     Min=UE.Subtract:Float(PrimaryOperand=^Random.Volume.Base, AdditionalOperands=@n2.Value).Out
    ).Value)
   ArrayRandomGet_0 = Array.Random Get:WaveAsset:Array(In Array=^Sounds, Next=^UE.Source.OnPlay)
   n3 = AbsoluteValue.Abs:Float(Input=^Random.Pitch.Spread)
   WavePlayer_0 = UE.Wave Player:Mono(
    Play=@ArrayRandomGet_0.On Next
    Wave Asset=@ArrayRandomGet_0.Value
    Pitch Shift=UE.RandomFloat(
     Next=^UE.Source.OnPlay
     Max=UE.Add:Float(PrimaryOperand=^Random.Pitch.Base, AdditionalOperands=@n3.Value).Out
     Min=UE.Subtract:Float(PrimaryOperand=^Random.Pitch.Base, AdditionalOperands=@n3.Value).Out
    ).Value
   ) {also {version [1,1]}}
   OnePoleLowPassFilter_0 = UE.One-Pole Low Pass Filter:Audio(
    Cutoff Frequency=MapRange.MapRange:Float(In=@n1.Out) {also {defaults {In Range A=0.75, In Range B=0.9, Out Range A=20000, Out Range B=150}}}.Out Value
    In=UE.Multiply:Audio by Float(AdditionalOperands=@DecibelstoLinearGain_0.Linear Gain, PrimaryOperand=@WavePlayer_0.Out Mono).Out
   )
   Multiply_0 = UE.Multiply:Audio by Float(
    AdditionalOperands=MapRange.MapRange:Float(In=@n1.Out) {also {defaults {In Range A=0.75, In Range B=0.9, Out Range A=1, Out Range B=0.2}}}.Out Value
    PrimaryOperand=@OnePoleLowPassFilter_0.Out
   )
  ]
  links [{dst="@UE.OutputFormat.Mono.Audio:0", src=Multiply_0.Out}, {dst="@UE.Source.OneShot.OnFinished", src=WavePlayer_0.On Finished}]
 }]
properties {
 assetImportData=&self:AssetImportData
 attenuationSettings=&/ATT_DragonCart_Orb_Laser_Fire_3P
 maxDistance=84000
 sampleRate=48000
 soundClassObject=&Fort_Events_Gameplay
 busSends [{audioBus=/DragonCartNarrative/Assets/Audio/OrbScan/AB_DragonCart_OrbMarkingActivate, sendLevel=0.2}]
}
