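# One-shot mono MetaSound for the dragon cart orb laser projectile, playing a whip-by as the projectile closes on the listener.
# On play it picks a random whip wave, which starts when attenuation distance falls past the activation threshold.
# Output is scaled by a distance volume curve and an affiliation mix.
# A delay of wave length plus padding ends the source if the whip never fires; playback resets it and finishes normally.
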MetaSoundSource /DragonCartNarrative/Assets/Audio/Laser/MSS_DragonCart_Laser_Prj
inputs [
 Character.Affiliation {type=Enum:FortAudioAffiliation}
 Character.Health {type=Float}
 Character.UniqueID {type=Int32}
 Sounds {type=WaveAsset:Array}
 UE.Attenuation.Distance {type=Float}
 UE.Source.OnPlay {type=Trigger}
 WhipActivationThreshold {type=Float}
]
interfaces [Fort.Character:1.0, UE.Attenuation:1.0, UE.OutputFormat.Mono:1.0, UE.Source.OneShot:1.0, UE.Source:1.1]
outputs [
 "UE.OutputFormat.Mono.Audio:0" {type=Audio}
 UE.Source.OneShot.OnFinished {type=Trigger}
]
pages [Default {
  input_defaults {
   Character.Affiliation=0
   Character.Health=0
   Character.UniqueID=0
   UE.Attenuation.Distance=0
   UE.Source.OnPlay=null
   WhipActivationThreshold=4000
   Sounds [
    /DragonCartNarrative/Assets/Audio/Laser/WAV/DragonCart_Orb_Laser_Whip_01
    /DragonCartNarrative/Assets/Audio/Laser/WAV/DragonCart_Orb_Laser_Whip_02
    /DragonCartNarrative/Assets/Audio/Laser/WAV/DragonCart_Orb_Laser_Whip_03
   ]
  }
  nodes [
   ArrayRandomGet_0 = Array.Random Get:WaveAsset:Array(In Array=^Sounds, Next=^UE.Source.OnPlay)
   WavePlayer_0 = UE.Wave Player:Mono(
    Wave Asset=@ArrayRandomGet_0.Value
    Play=UE.TriggerOnThreshold:Float(In=^UE.Attenuation.Distance, Threshold=^WhipActivationThreshold) {also {defaults {Type=1}}}.Out
   ) {also {version [1,1]}}
   Multiply_0 = UE.Multiply:Audio by Float(
    AdditionalOperands=&MSP_AffiliationMixFloat(Affiliation=^Character.Affiliation) {also {defaults {Enemy=0, Non-Enemy=0}}}.Out
    PrimaryOperand=UE.Multiply:Audio by Float(
     AdditionalOperands=&MSP_Attenuation_Volume_Curve(Distance=^UE.Attenuation.Distance) {also {defaults {AttenuationCurves=&/WTB_DragonCart_OrbLaser_Prj_Distance, FalloffDistance=10000, InnerRadius=500}}}.VolumeMultiplier
     PrimaryOperand=@WavePlayer_0.Out Mono
    ).Out
   )
   GetWaveDuration_0 = "UE.Get Wave Duration: "(Wave=@ArrayRandomGet_0.Value)
   n1 = InitVariable.Trigger()
   TriggerDelay_0 = UE.Trigger Delay(
    Delay Time=UE.Add:Time(PrimaryOperand=@GetWaveDuration_0.Duration) {also {defaults {AdditionalOperands=1}}}.Out
    In=@ArrayRandomGet_0.On Next
    Reset=VariableAccessor.Trigger(Variable=VariableMutator.Trigger(Variable=@n1.Variable, Value=@WavePlayer_0.On Play).Variable).Value
   )
   TriggerAnyTriggerAny2_0 = "TriggerAny.Trigger Any (2)"(In 0=@TriggerDelay_0.Out, In 1=@WavePlayer_0.On Finished)
  ]
  links [{dst="@UE.OutputFormat.Mono.Audio:0", src=Multiply_0.Out}, {dst="@UE.Source.OneShot.OnFinished", src=TriggerAnyTriggerAny2_0.Out}]
 }]
properties {
 assetImportData=&self:AssetImportData
 attenuationSettings=&/ATT_DragonCart_Orb_Prj
 bOverrideConcurrency=true
 maxDistance=10500
 sampleRate=48000
 soundClassObject=&Fort_WeaponRanged_Projectile_Effects
 volume=2
 concurrencyOverrides {maxCount=2, resolutionRule=PreventNew}
}
