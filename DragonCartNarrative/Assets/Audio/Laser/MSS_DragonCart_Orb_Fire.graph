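# One-shot stereo MetaSound source for the DragonCart orb laser fire sound.
# On play, a random pitch shift is chosen and a shuffled pick from three fire variations goes to a stereo wave player.
# An enclosure worldization modulator is range-mapped to drive per-channel low-pass cutoff and gain, making the shot duller and quieter as that value rises.
# Volume is also routed through a DragonCart laser control bus, with a first-person-named attenuation asset applied.
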MetaSoundSource /DragonCartNarrative/Assets/Audio/Laser/MSS_DragonCart_Orb_Fire
inputs [
 Sounds {type=WaveAsset:Array}
 UE.Source.OnPlay {type=Trigger}
]
interfaces [UE.OutputFormat.Stereo:1.0, UE.Source.OneShot:1.0, UE.Source:1.1]
outputs [
 "UE.OutputFormat.Stereo.Audio:0" {type=Audio}
 "UE.OutputFormat.Stereo.Audio:1" {type=Audio}
 UE.Source.OneShot.OnFinished {type=Trigger}
]
pages [Default {
  input_defaults {
   UE.Source.OnPlay=null
   Sounds [
    /DragonCartNarrative/Assets/Audio/Laser/WAV/DragonCart_Orb_Laser_Fire_01
    /DragonCartNarrative/Assets/Audio/Laser/WAV/DragonCart_Orb_Laser_Fire_02
    /DragonCartNarrative/Assets/Audio/Laser/WAV/DragonCart_Orb_Laser_Fire_03
   ]
  }
  nodes [
   n1 = Modulation.GetModulatorValue() {also {defaults {Modulator=&AudioWorldization_Bus_Enclosure}}}
   n2 = MapRange.MapRange:Float(In=@n1.Out) {also {defaults {In Range A=0.75, In Range B=0.9, Out Range A=1, Out Range B=0.2}}}
   n3 = MapRange.MapRange:Float(In=@n1.Out) {also {defaults {In Range A=0.75, In Range B=0.9, Out Range A=20000, Out Range B=150}}}
   n4 = UE.RandomFloat(Next=^UE.Source.OnPlay) {also {defaults {Max=0.1, Min=-0.1}}}
   n5 = Array.Shuffle:WaveAsset:Array(In Array=^Sounds, Next=@n4.On Next) {also {defaults {Enable Shared State="<MetaSoundBuilderResult.SUCCEEDED: 0>"}}}
   WavePlayer_0 = UE.Wave Player:Stereo(Play=@n5.On Next, Wave Asset=@n5.Value, Pitch Shift=@n4.Value) {also {version [1,1]}}
   OnePoleLowPassFilter_1 = UE.One-Pole Low Pass Filter:Audio(Cutoff Frequency=@n3.Out Value, In=@WavePlayer_0.Out Right)
   Multiply_0 = UE.Multiply:Audio by Float(AdditionalOperands=@n2.Out Value, PrimaryOperand=@OnePoleLowPassFilter_1.Out)
   OnePoleLowPassFilter_0 = UE.One-Pole Low Pass Filter:Audio(Cutoff Frequency=@n3.Out Value, In=@WavePlayer_0.Out Left)
   Multiply_1 = UE.Multiply:Audio by Float(AdditionalOperands=@n2.Out Value, PrimaryOperand=@OnePoleLowPassFilter_0.Out)
  ]
  links [
   {dst="@UE.OutputFormat.Stereo.Audio:1", src=Multiply_0.Out}
   {dst="@UE.OutputFormat.Stereo.Audio:0", src=Multiply_1.Out}
   {dst="@UE.Source.OneShot.OnFinished", src=WavePlayer_0.On Finished}
  ]
 }]
properties {
 assetImportData=&self:AssetImportData
 attenuationSettings=&/ATT_DragonCart_Orb_Laser_Fire_1P
 maxDistance=84000
 numChannels=2
 outputFormat=Stereo
 sampleRate=48000
 soundClassObject=&Fort_WeaponRanged_Fire_3P_PSM
 volume=2
 modulationSettings {
  volumeRouting=Union
  volumeModulationDestination {bEnableModulation=true, modulators [&/CB_DragonCart_Laser]}
 }
}
