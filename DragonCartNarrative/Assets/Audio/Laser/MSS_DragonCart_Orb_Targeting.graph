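# Mono MetaSound source for the DragonCart orb laser targeting sound, using the third-person laser fire attenuation.
# On play, it picks randomly from the Sounds array, which holds one wave by default, and feeds a wave player.
# The AudioWorldization enclosure modulation bus drives a one-pole low-pass and a gain stage, muffling and quieting the sound as enclosure rises through the upper part of its range.
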MetaSoundSource /DragonCartNarrative/Assets/Audio/Laser/MSS_DragonCart_Orb_Targeting
inputs [
 Sounds {type=WaveAsset:Array}
 UE.Source.OnPlay {type=Trigger}
]
interfaces [UE.OutputFormat.Mono:1.0, UE.Source:1.1]
outputs ["UE.OutputFormat.Mono.Audio:0" {type=Audio}]
pages [Default {
  input_defaults {UE.Source.OnPlay=null, Sounds [/DragonCartNarrative/Assets/Audio/Laser/WAV/DragonCart_Orb_Laser_Targetting_Orb_01]}
  nodes [
   n1 = Modulation.GetModulatorValue() {also {defaults {Modulator=&AudioWorldization_Bus_Enclosure}}}
   ArrayRandomGet_0 = Array.Random Get:WaveAsset:Array(In Array=^Sounds, Next=^UE.Source.OnPlay)
   WavePlayer_0 = UE.Wave Player:Mono(Play=@ArrayRandomGet_0.On Next, Wave Asset=@ArrayRandomGet_0.Value) {also {version [1,1]}}
   OnePoleLowPassFilter_0 = UE.One-Pole Low Pass Filter:Audio(
    Cutoff Frequency=MapRange.MapRange:Float(In=@n1.Out) {also {defaults {In Range A=0.75, In Range B=0.9, Out Range A=20000, Out Range B=150}}}.Out Value
    In=@WavePlayer_0.Out Mono
   )
   Multiply_0 = UE.Multiply:Audio by Float(
    AdditionalOperands=MapRange.MapRange:Float(In=@n1.Out) {also {defaults {In Range A=0.75, In Range B=0.9, Out Range A=1, Out Range B=0.2}}}.Out Value
    PrimaryOperand=@OnePoleLowPassFilter_0.Out
   )
  ]
  links [{dst="@UE.OutputFormat.Mono.Audio:0", src=Multiply_0.Out}]
 }]
properties {
 assetImportData=&self:AssetImportData
 attenuationSettings=&/ATT_DragonCart_Orb_Laser_Fire_3P
 bLooping=true
 duration=10000
 maxDistance=84000
 sampleRate=48000
 soundClassObject=&Fort_Events_Gameplay
 volume=1.5
}
