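# Looping mono MetaSound for the DragonCart orb's targeting rotation, playing a wave from the Sounds array on play.
# Each CheckRotation trigger compares Rotation with a stored value, using variables and a deferred trigger to retain the previous reading.
# The result routes two gain and pitch-shift targets, smoothed by interpolation, so the loop reacts to turning.
# The enclosure worldization bus modulator then drives a low-pass cutoff and final attenuation, muffling the sound as its value rises.
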MetaSoundSource /DragonCartNarrative/Assets/Audio/Laser/MSS_DragonCart_Orb_Targeting_Rotate_LP
inputs [
 CheckRotation {type=Trigger}
 MasterGain {type=Float}
 Movement.Speed {type=Float}
 Movement.SpeedAzimuth {type=Float}
 Movement.SpeedElevation {type=Float}
 Rotation {type=Float}
 Sounds {type=WaveAsset:Array}
 UE.Source.OnPlay {type=Trigger}
 UE.Source.Orientation.Azimuth {type=Float}
 UE.Source.Orientation.Elevation {type=Float}
]
interfaces [Fort.Movement:1.0, UE.OutputFormat.Mono:1.0, UE.Source.Orientation:1.0, UE.Source:1.1]
outputs ["UE.OutputFormat.Mono.Audio:0" {type=Audio}]
pages [Default {
  input_defaults {
   CheckRotation=null
   MasterGain=2
   Movement.Speed=0
   Movement.SpeedAzimuth=-100.7999
   Movement.SpeedElevation=0
   Rotation=-115.2
   UE.Source.OnPlay=null
   UE.Source.Orientation.Azimuth=0.396
   UE.Source.Orientation.Elevation=0
   Sounds [/DragonCartNarrative/Assets/Audio/Laser/WAV/DragonCart_Orb_Rotation_LP]
  }
  nodes [
   InitVariable.Float()
   n1 = Modulation.GetModulatorValue() {also {defaults {Modulator=&AudioWorldization_Bus_Enclosure}}}
   n2 = InitVariable.Float()
   n3 = VariableDeferredAccessor.Trigger(Variable=InitVariable.Trigger().Variable)
   n4 = Value.Value:Float(Target Value=^Rotation, Set=@n3.Value)
   n5 = TriggerCompare.Clamp:Float(
    Compare=^CheckRotation
    A=^Rotation
    B=VariableAccessor.Float(Variable=VariableMutator.Float(Variable=@n2.Variable, Value=@n4.Output Value).Variable).Value
   )
   TriggerRouteTriggerRouteFloat2_1 = "TriggerRoute.Trigger Route (Float, 2):Float"(Set 1=@n5.False, Set 0=@n5.True) {also {defaults {Value 0=0.5, Value 1=1}}}
   n6 = UE.InterpTo:Audio(Target=@TriggerRouteTriggerRouteFloat2_1.Value) {also {defaults {Interp Time=1}}}
   ArrayRandomGet_0 = Array.Random Get:WaveAsset:Array(In Array=^Sounds, Next=^UE.Source.OnPlay)
   TriggerRouteTriggerRouteFloat2_0 = "TriggerRoute.Trigger Route (Float, 2):Float"(Set 1=@n5.False, Set 0=@n5.True) {also {defaults {Value 0=0, Value 1=3}}}
   WavePlayer_0 = UE.Wave Player:Mono(
    Play=@ArrayRandomGet_0.On Next
    Wave Asset=@ArrayRandomGet_0.Value
    Pitch Shift=UE.InterpTo:Audio(Target=@TriggerRouteTriggerRouteFloat2_0.Value) {also {defaults {Interp Time=3}}}.Value
   ) {also {defaults {Loop="<MetaSoundBuilderResult.SUCCEEDED: 0>"}, version [1,1]}}
   OnePoleLowPassFilter_0 = UE.One-Pole Low Pass Filter:Audio(
    Cutoff Frequency=MapRange.MapRange:Float(In=@n1.Out) {also {defaults {In Range A=0.75, In Range B=0.9, Out Range A=20000, Out Range B=150}}}.Out Value
    In=UE.Multiply:Audio by Float(
     AdditionalOperands=^MasterGain
     PrimaryOperand=UE.Multiply:Audio by Float(AdditionalOperands=@n6.Value, PrimaryOperand=@WavePlayer_0.Out Mono).Out
    ).Out
   )
   Multiply_0 = UE.Multiply:Audio by Float(
    AdditionalOperands=MapRange.MapRange:Float(In=@n1.Out) {also {defaults {In Range A=0.75, In Range B=0.9, Out Range A=1, Out Range B=0.2}}}.Out Value
    PrimaryOperand=@OnePoleLowPassFilter_0.Out
   )
   VariableMutator.Trigger(Value=^CheckRotation, Variable=@n3.Variable)
  ]
  links [{dst="@UE.OutputFormat.Mono.Audio:0", src=Multiply_0.Out}]
 }]
properties {
 assetImportData=&self:AssetImportData
 attenuationSettings=&/ATT_DragonCart_Orb_Laser_Fire_3P
 bLooping=true
 duration=10000
 maxDistance=84000
 sampleRate=48000
 soundClassObject=&Fort_WeaponRanged_Projectile_Effects
}
